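# Base character material for the Rage Debris backpack cosmetic, with a glowing numeric display screen that brightens on eliminations.
# MF_NumberDisplay digits, placed by offset and scaled UVs, combine with a backlit screen mask, tinted by DigitColor.
# The result maxes into the parent material's emissive between the shared PreFX and PostFX functions.
# A custom primitive data timestamp drives a timed fade-in, hold and fade-out between idle and elimination strengths.
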Material /BRCosmetics/Accessories/FORT_Backpacks/Backpack_M_MED_RageDebris/Materials/M_RageDebris_Backpack_Base
graph {
 properties {
  bUseMaterialAttributes=true
  bUsedWithMeshDeformer=true
  bUsedWithMorphTargets=true
  bUsedWithSkeletalMesh=true
  pixelDepthOffsetMode=PDOM_Legacy
  shadingModel=MSM_FromMaterialExpression
  usedShadingModels=Default Lit | Subsurface | Clear Coat | Subsurface Profile | Cloth
 }
 nodes [
  n1 = MaterialFunctionCall() {materialFunction=&MF_CharacterParent_2023_PreFX}
  MaterialFunctionCall(Material=NamedRerouteUsage(), M=@n1.M, Pre Skinned Local Position=@n1.M, Distance Blend=@n1.M, Fresnel=@n1.M) {materialFunction=&MF_CharacterParent_2023_PostFX}
  NamedRerouteDeclaration(None=@n1.Result) {name=IN_Material, nodeColor=<rgba 0.9219 0 1 1>}
  n2 = GetMaterialAttributes(None=NamedRerouteUsage()) {attributeGetTypes [#1]}
  n3 = Add(A=TextureCoordinate(), B=Constant2Vector() {g=-0.13, r=-0.6185})
  n4 = Constant2Vector() {g=13, r=1.75}
  n5 = MaterialFunctionCall!(UVs=Saturate(None=Multiply(A=@n3, B=@n4))) {materialFunction=&/MF_NumberDisplay}
  n6 = VectorParameter() {group=RageDebris, parameterName=ScreenMask_UVScale, defaultValue=<rgba 1 1 1 0>}
  n7 = Multiply(A=TextureCoordinate(), B=AppendVector(A=@n6.R, B=@n6.R))
  n8 = VectorParameter() {group=RageDebris, parameterName=ScreenMask_UVOffset}
  n9 = AppendVector(A=@n8.R, B=@n8.R)
  n10 = TextureSampleParameter2D(UVs=Saturate(None=Add(A=@n7, B=@n9))) {
   group=RageDebris
   parameterName=ScreenMask
   samplerType=SAMPLERTYPE_Masks
   texture=/BRCosmetics/Accessories/FORT_Backpacks/Backpack_M_MED_RageDebris/Textures/T_RageDebris_Backpack_TechArt_PROTO
  }
  n11 = LinearInterpolate(
   A=ScalarParameter() {defaultValue=0.15, group=RageDebris, parameterName=Backlight_IdleStrength}
   B=ScalarParameter() {defaultValue=0.5, group=RageDebris, parameterName=Backlight_ElimStrength}
   Alpha=NamedRerouteUsage()
  )
  n12 = Multiply(A=@n10.R, B=Max(A=Multiply(A=@n10.G, B=@n11), B=@n10.B))
  n13 = Multiply(
   A=Multiply(A=Max(A=@n5.Result, B=@n12), B=VectorParameter() {group=NumberDisplay, parameterName=DigitColor, defaultValue=<rgba 0.93 0 0 1>}.RGB)
   B=LinearInterpolate(
    A=ScalarParameter() {defaultValue=1, group=RageDebris, parameterName=Emissive_IdleStrength}
    B=ScalarParameter() {defaultValue=3, group=RageDebris, parameterName=Emissive_ElimStrength}
    Alpha=NamedRerouteUsage()
   )
  )
  NamedRerouteDeclaration(None=SetMaterialAttributes(MaterialAttributes=@n2.MaterialAttributes, Emissive Color=Max(A=@n2.Emissive Color, B=@n13)) {attributeSetTypes [#1]}) {name=OUT_Material, nodeColor=<rgba 0 1 0.6016 1>}
  NamedRerouteDeclaration(None=@n1.Tech Art Mask) {name=TechArt_Mask, nodeColor=<rgba 1 0.2812 0 1>}
  n14 = ScalarParameter() {bUseCustomPrimitiveData=true, group=RageDebris, parameterName=TimeStamp_CPD, primitiveDataIndex=4}
  n15 = Abs(None=Subtract(A=Time(), B=@n14))
  n16 = ScalarParameter() {defaultValue=0.5, group=RageDebris, parameterName=EffectSpeed_In}
  n17 = Multiply(A=Saturate(None=Divide(A=@n15, B=@n16)), B=Saturate(None=@n14))
  n18 = Time()
  n19 = Add(
   A=ScalarParameter() {bUseCustomPrimitiveData=true, group=RageDebris, parameterName=TimeStamp_CPD, primitiveDataIndex=4}
   B=ScalarParameter() {defaultValue=7, group=RageDebris, parameterName=EffectDelay}
  ) {constB=5}
  n20 = Saturate(None=Divide(A=Subtract(A=@n18, B=@n19), B=ScalarParameter() {defaultValue=2, group=RageDebris, parameterName=EffectSpeed_Out}))
  NamedRerouteDeclaration(None=Saturate(None=Subtract(A=@n17, B=@n20))) {name=TimeStampAnimation, nodeColor=<rgba 0.7812 0 1 1>}
 ]
}
