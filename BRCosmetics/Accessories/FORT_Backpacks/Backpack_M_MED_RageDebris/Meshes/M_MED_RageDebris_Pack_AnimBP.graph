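# Anim Blueprint for the RageDebris backpack cosmetic that adds physics sway to its dangling wires and cable.
# Five single-body Anim Dynamics nodes run in series on the input pose, covering two bones per side wire and one cable bone, each with angular limits and custom gravity.
# Wires simulate root-relative.
# On customization load, DynamicDamping is set by front-end state, but the nodes show fixed damping and no read of it.
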AnimBlueprint /BRCosmetics/Accessories/FORT_Backpacks/Backpack_M_MED_RageDebris/Meshes/M_MED_RageDebris_Pack_AnimBP
target_skeleton=&/M_MED_RageDebris_Pack_Skeleton
graphs [
 AnimGraph {nodes [
   n1 = Anim Dynamics - Bone: dyn_wire_r_a!(ComponentPose=Local To Component(LocalPose=Input Pose()), Alpha=1, AngularDampingOverride:float=0.97, ExternalForce="0,0,0") -> (Pose) {
    angularDampingOverride=0.97
    bOverrideAngularDamping=true
    bUseGravityOverride=true
    simulationSpace=RootRelative
    alphaBoolBlend {blendOption=Linear}
    boundBone {boneName=dyn_wire_r_a}
    componentAppliedLinearAccClamp=<4000 4000 4000>
    componentLinearAccScale=<0.6 0.6 0.6>
    componentLinearVelScale=<0.6 0.6 0.6>
    gravityOverride {z=-200}
    physicsBodyDefinitions [{
      boundBone {boneName=dyn_wire_r_a}
      boxExtents=<10 10 10>
      constraintSetup {angularLimitsMax=<0 30 15>, angularLimitsMin=<0 -30 -15>, angularTarget=<0 0 0>}
      localJointOffset=<3 -0.2 0>
     }]
   }
   n2 = Anim Dynamics - Bone: dyn_wire_r_b!(ComponentPose=@n1, Alpha=1, AngularDampingOverride:float=0.97, ExternalForce="0,0,0") -> (Pose) {
    angularDampingOverride=0.97
    bOverrideAngularDamping=true
    bUseGravityOverride=true
    simulationSpace=RootRelative
    alphaBoolBlend {blendOption=Linear}
    boundBone {boneName=dyn_wire_r_b}
    componentAppliedLinearAccClamp=<4000 4000 4000>
    componentLinearAccScale=<0.6 0.6 0.6>
    componentLinearVelScale=<0.6 0.6 0.6>
    gravityOverride {z=-200}
    physicsBodyDefinitions [{
      boundBone {boneName=dyn_wire_r_b}
      boxExtents=<10 10 10>
      constraintSetup {angularLimitsMax=<0 20 10>, angularLimitsMin=<0 -20 0>, angularTarget=<0 0 0>}
      localJointOffset=<2 0 0.2>
     }]
   }
   n3 = Anim Dynamics - Bone: dyn_wire_l_a!(ComponentPose=@n2, Alpha=1, AngularDampingOverride:float=0.97, ExternalForce="0,0,0") -> (Pose) {
    angularDampingOverride=0.97
    bOverrideAngularDamping=true
    bUseGravityOverride=true
    simulationSpace=RootRelative
    alphaBoolBlend {blendOption=Linear}
    boundBone {boneName=dyn_wire_l_a}
    componentAppliedLinearAccClamp=<4000 4000 4000>
    componentLinearAccScale=<0.6 0.6 0.6>
    componentLinearVelScale=<0.6 0.6 0.6>
    gravityOverride {z=-200}
    physicsBodyDefinitions [{
      boundBone {boneName=dyn_wire_l_a}
      boxExtents=<10 10 10>
      constraintSetup {angularLimitsMax=<0 30 15>, angularLimitsMin=<0 -30 -15>, angularTarget=<0 0 0>}
      localJointOffset=<2 0 -0.2>
     }]
   }
   n4 = Anim Dynamics - Bone: dyn_wire_l_b!(ComponentPose=@n3, Alpha=1, AngularDampingOverride:float=0.97, ExternalForce="0,0,0") -> (Pose) {
    angularDampingOverride=0.97
    bOverrideAngularDamping=true
    bUseGravityOverride=true
    simulationSpace=RootRelative
    alphaBoolBlend {blendOption=Linear}
    boundBone {boneName=dyn_wire_l_b}
    componentAppliedLinearAccClamp=<4000 4000 4000>
    componentLinearAccScale=<0.6 0.6 0.6>
    componentLinearVelScale=<0.6 0.6 0.6>
    gravityOverride {z=-200}
    physicsBodyDefinitions [{
      boundBone {boneName=dyn_wire_l_b}
      boxExtents=<10 10 10>
      constraintSetup {angularLimitsMax=<0 20 10>, angularLimitsMin=<0 -20 0>, angularTarget=<0 0 0>}
      localJointOffset=<2 0.2 0>
     }]
   }
   n5 = Anim Dynamics - Bone: dyn_cable_a!(ComponentPose=@n4, Alpha=1, AngularDampingOverride:float=0.97, ExternalForce="0,0,0") -> (Pose) {
    angularDampingOverride=0.97
    bOverrideAngularDamping=true
    bUseGravityOverride=true
    alphaBoolBlend {blendOption=Linear}
    boundBone {boneName=dyn_cable_a}
    componentAppliedLinearAccClamp=<4000 4000 4000>
    componentLinearAccScale=<0.5 0.5 0.5>
    componentLinearVelScale=<0.5 0.5 0.5>
    gravityOverride {z=-400}
    physicsBodyDefinitions [{
      boundBone {boneName=dyn_cable_a}
      boxExtents=<10 10 10>
      constraintSetup {angularLimitsMax=<0 45 0>, angularLimitsMin=<0 -45 -90>, angularTarget=<0 0 0>}
      localJointOffset=<0 0 -3>
     }]
   }
   Output Pose(Result=Component To Local(ComponentPose=@n5)) {layerGroup=None}
  ]}
 EventGraph {nodes [
   Event OnCharacterCustomizationLoaded()
   > Set DynamicDamping(DynamicDamping:double=SelectFloat(A=0.97, B=0.95, bPickA=Get bIsInFrontEnd())) -> (Output_Get:double)
  ]}
]
variables [DynamicDamping {type=number}]
